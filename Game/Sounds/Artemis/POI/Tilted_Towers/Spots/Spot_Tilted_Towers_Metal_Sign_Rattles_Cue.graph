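# Ambient one-shot cue for metal sign rattles at the Tilted Towers point of interest.
# A Random node picks one of six equally weighted metal movement waves and feeds a Modulator that varies playback, with a widened lower pitch bound.
# The cue plays on the ambient one-shot sound class at reduced volume, using a less-positional log attenuation asset.
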SoundCue /Game/Sounds/Artemis/POI/Tilted_Towers/Spots/Spot_Tilted_Towers_Metal_Sign_Rattles_Cue
root=Modulator_0
nodes [
 {id=Modulator_0, children [Random_1], properties {pitchMin=0.9}}
 {
  id=Random_1
  children [WavePlayer_0, WavePlayer_1, WavePlayer_2, WavePlayer_8, WavePlayer_9, WavePlayer_10]
  properties {weights [1,1,1,1,1,1]}
 }
 {id=WavePlayer_0, properties {soundWaveAssetPtr=&/Spot_Tilted_Towers_Metal_Movement_06}}
 {id=WavePlayer_1, properties {soundWaveAssetPtr=&/Spot_Tilted_Towers_Metal_Movement_01}}
 {id=WavePlayer_10, properties {soundWaveAssetPtr=&/Spot_Tilted_Towers_Metal_Movement_05}}
 {id=WavePlayer_2, properties {soundWaveAssetPtr=&/Spot_Tilted_Towers_Metal_Movement_02}}
 {id=WavePlayer_8, properties {soundWaveAssetPtr=&/Spot_Tilted_Towers_Metal_Movement_03}}
 {id=WavePlayer_9, properties {soundWaveAssetPtr=&/Spot_Tilted_Towers_Metal_Movement_04}}
]
properties {
 attenuationSettings=/Game/Sounds/Attenuation/Fort_TileAmbient_OneShot_LessPositional_Log_Att
 maxDistance=6200
 soundClassObject=&Fort_Ambient_OneShots
 volumeMultiplier=0.501
 attenuationOverrides {falloffDistance=6000, stereoSpread=10000, attenuationShapeExtents {x=2000}}
}
